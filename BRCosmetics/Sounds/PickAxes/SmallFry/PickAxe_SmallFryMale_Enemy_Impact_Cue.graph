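# Sound cue for the male SmallFry cosmetic pickaxe's impact on an enemy.
# A Random node picks one of six equally weighted impact waves and feeds it through a root Modulator.
# It uses the shared pickaxe impact attenuation and property sheet, the Fort_Impacts sound class, and player melee weapon concurrency.
# The cue is primed on load.
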SoundCue /BRCosmetics/Sounds/PickAxes/SmallFry/PickAxe_SmallFryMale_Enemy_Impact_Cue
root=Modulator_0
nodes [
 {id=Modulator_0, children [Random_3]}
 {
  id=Random_3
  children [WavePlayer_36, WavePlayer_37, WavePlayer_38, WavePlayer_39, WavePlayer_40, WavePlayer_41]
  properties {weights [1,1,1,1,1,1]}
 }
 {id=WavePlayer_36, properties {soundWaveAssetPtr=&/Pickaxe_SmallfryMale_EnemyImpact_01}}
 {id=WavePlayer_37, properties {soundWaveAssetPtr=&/Pickaxe_SmallfryMale_EnemyImpact_02}}
 {id=WavePlayer_38, properties {soundWaveAssetPtr=&/Pickaxe_SmallfryMale_EnemyImpact_03}}
 {id=WavePlayer_39, properties {soundWaveAssetPtr=&/Pickaxe_SmallfryMale_EnemyImpact_04}}
 {id=WavePlayer_40, properties {soundWaveAssetPtr=&/Pickaxe_SmallfryMale_EnemyImpact_05}}
 {id=WavePlayer_41, properties {soundWaveAssetPtr=&/Pickaxe_SmallfryMale_EnemyImpact_06}}
]
properties {
 attenuationSettings=/Game/Sounds/Attenuation/Fort_Pickaxe_Impact_Att
 audioPropertiesSheet=/BRCosmetics/Sounds/Settings/PropertySheets/Pickaxe/APS_Cosmetic_Pickaxe_Impact
 bPrimeOnLoad=true
 maxDistance=4300
 soundClassObject=&Fort_Impacts
 volumeMultiplier=0.4875
 attenuationOverrides {
  bAttenuateWithLPF=true
  distanceAlgorithm=Logarithmic
  falloffDistance=800
  lPFRadiusMax=800
  lPFRadiusMin=400
  attenuationShapeExtents {x=300}
 }
 concurrencyOverrides {bLimitToOwner=true, maxCount=2, resolutionRule=StopOldest}
 concurrencySet [&Player_Melee_Weapon_Conc]
}
