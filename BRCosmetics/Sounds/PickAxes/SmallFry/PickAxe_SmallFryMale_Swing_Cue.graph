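# Sound cue for the swing of the SmallFry Male cosmetic pickaxe.
# It mixes two modulated layers: a random SmallFry swing through close pickaxe attenuation and a random shared distant fire axe swing through medium attenuation.
# Both layers are scaled down in the mixer, and an owner-limited concurrency override stops the oldest swing when its small voice cap is exceeded.
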SoundCue /BRCosmetics/Sounds/PickAxes/SmallFry/PickAxe_SmallFryMale_Swing_Cue
root=Mixer_0
nodes [
 {id=Attenuation_0, children [Modulator_1], properties {attenuationSettings=/Game/Sounds/Attenuation/Fort_Weapon_Pickaxe_Close_Att}}
 {id=Attenuation_1, children [Modulator_0], properties {attenuationSettings=/Game/Sounds/Attenuation/Fort_Weapon_Pickaxe_Medium_Att}}
 {id=Mixer_0, children [Attenuation_0, Attenuation_1], properties {inputVolume [0.175,0.2]}}
 {id=Modulator_0, children [Random_0]}
 {id=Modulator_1, children [Random_7]}
 {id=Random_0, children [WavePlayer_1,WavePlayer_2,WavePlayer_4,WavePlayer_6,WavePlayer_7,WavePlayer_8], properties {weights [1,1,1,1,1,1]}}
 {
  id=Random_7
  children [WavePlayer_63, WavePlayer_64, WavePlayer_65, WavePlayer_66, WavePlayer_67, WavePlayer_68]
  properties {weights [1,1,1,1,1,1]}
 }
 {id=WavePlayer_1, properties {soundWaveAssetPtr=/Game/Athena/Sounds/Weapons/Fire_Axe_Swing_Distant_04}}
 {id=WavePlayer_2, properties {soundWaveAssetPtr=/Game/Athena/Sounds/Weapons/Fire_Axe_Swing_Distant_01}}
 {id=WavePlayer_4, properties {soundWaveAssetPtr=/Game/Athena/Sounds/Weapons/Fire_Axe_Swing_Distant_02}}
 {id=WavePlayer_6, properties {soundWaveAssetPtr=/Game/Athena/Sounds/Weapons/Fire_Axe_Swing_Distant_03}}
 {id=WavePlayer_63, properties {soundWaveAssetPtr=&/Pickaxe_SmallfryMale_Swing_01}}
 {id=WavePlayer_64, properties {soundWaveAssetPtr=&/Pickaxe_SmallfryMale_Swing_02}}
 {id=WavePlayer_65, properties {soundWaveAssetPtr=&/Pickaxe_SmallfryMale_Swing_03}}
 {id=WavePlayer_66, properties {soundWaveAssetPtr=&/Pickaxe_SmallfryMale_Swing_04}}
 {id=WavePlayer_67, properties {soundWaveAssetPtr=&/Pickaxe_SmallfryMale_Swing_05}}
 {id=WavePlayer_68, properties {soundWaveAssetPtr=&/Pickaxe_SmallfryMale_Swing_06}}
 {id=WavePlayer_7, properties {soundWaveAssetPtr=/Game/Athena/Sounds/Weapons/Fire_Axe_Swing_Distant_05}}
 {id=WavePlayer_8, properties {soundWaveAssetPtr=/Game/Athena/Sounds/Weapons/Fire_Axe_Swing_Distant_06}}
]
properties {
 audioPropertiesSheet=/BRCosmetics/Sounds/Settings/PropertySheets/Pickaxe/APS_Cosmetic_Pickaxe_Swing
 bOverrideConcurrency=true
 bPrimeOnLoad=true
 maxDistance=3000
 soundClassObject=&Fort_WeaponMelee_PSM
 volumeMultiplier=1
 concurrencyOverrides {bLimitToOwner=true, maxCount=2, resolutionRule=StopOldest}
 concurrencySet [&Player_Melee_Weapon_Conc]
}
